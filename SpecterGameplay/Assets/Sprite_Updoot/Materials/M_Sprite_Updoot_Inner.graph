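# Masked unlit material for Niagara mesh particles, apparently the inner part of the Updoot sprite, with chevron stripes cut out of it.
# A view-to-normal dot product through Power blends Edge Color into Base Color, scaled by Emissive Multiplier, and also drives DitherTemporalAA to dither the silhouette.
# The Chevrons reroute, which multiplies that dither, joins oppositely rotated UV SmoothStep bands at the U midpoint to form two V-shaped gaps.
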Material /SpecterGameplay/Assets/Sprite_Updoot/Materials/M_Sprite_Updoot_Inner
graph {
 properties {
  bAutomaticallySetUsageInEditor=false
  bUsedWithNiagaraMeshParticles=true
  blendMode=BLEND_Masked
  shadingModel=MSM_Unlit
  usedShadingModels=Unlit
 }
 nodes [
  n1 = Power(Base=DotProduct(A=CameraVectorWS(), B=VertexNormalWS()))
  Multiply(
   A=LinearInterpolate(
    A=VectorParameter() {parameterName=Edge Color, defaultValue=<rgba 0.0908 0.3229 0.22 1>}.RGB
    B=VectorParameter() {parameterName=Base Color, defaultValue=<rgba 0 0.7652 1 1>}.RGB
    Alpha=@n1
   )
   B=ScalarParameter() {defaultValue=2, parameterName=Emissive Multiplier}
  )
  Multiply(
   A=NamedRerouteUsage()
   B=MaterialFunctionCall(Alpha Threshold=@n1) {materialFunction=&DitherTemporalAA}.Result
  )
  n2 = TextureCoordinate()
  n3 = ComponentMask(None=Rotator(Coordinate=@n2, Time=Constant() {r=2})) {g=true}
  n4 = Multiply(
   A=Add(A=SmoothStep(Value=@n3) {constMax=0.7, constMin=0.65}, B=SmoothStep(Value=@n3) {constMax=0.6, constMin=0.65})
   B=Add(A=SmoothStep(Value=@n3) {constMax=0.6, constMin=0.55}, B=SmoothStep(Value=@n3) {constMax=0.5, constMin=0.55})
  )
  n5 = ComponentMask(None=Rotator(Coordinate=@n2, Time=Constant() {r=-2})) {g=true}
  n6 = Multiply(
   A=Add(A=SmoothStep(Value=@n5) {constMax=0.7, constMin=0.65}, B=SmoothStep(Value=@n5) {constMax=0.6, constMin=0.65})
   B=Add(A=SmoothStep(Value=@n5) {constMax=0.6, constMin=0.55}, B=SmoothStep(Value=@n5) {constMax=0.5, constMin=0.55})
  )
  n7 = OneMinus(None=Step(Y=ComponentMask(None=@n2) {r=true}) {constX=0.5})
  NamedRerouteDeclaration(None=Saturate(None=LinearInterpolate(A=@n4, B=@n6, Alpha=@n7))) {name=Chevrons, nodeColor=<rgba 0.8047 1 0 1>}
 ]
}
