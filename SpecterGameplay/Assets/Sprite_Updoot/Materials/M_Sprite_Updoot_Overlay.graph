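# Masked, two-sided lit overlay material for the Updoot sprite mesh, flagged for Niagara mesh particles.
# View angle and vertex colors blend main, core and edge colors, layered with scrolling bubbles and a sugar-coating texture from the bubble map.
# Dynamic parameters drive eye and mouth closing through UV-half masks.
# Named reroutes organize color, emissive and opacity, alongside a camera-distance fade, sky-light tint, dithered sun specular mask and Updoot WPO function.
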Material /SpecterGameplay/Assets/Sprite_Updoot/Materials/M_Sprite_Updoot_Overlay
graph {
 properties {
  bAutomaticallySetUsageInEditor=false
  bUsedWithNiagaraMeshParticles=true
  blendMode=BLEND_Masked
  previewMesh=/SpecterGameplay/Assets/Sprite_Updoot/Meshes/Sprite_Updoot
  twoSided=true
  usedShadingModels=Default Lit
 }
 nodes [
  n1 = Sine(None=Divide(A=Time()) {constB=100})
  n2 = BumpOffset(
   Coordinate=Add(A=TextureCoordinate(), B=Divide(A=@n1) {constB=10}) {constB=0.3}
   Height=ScalarParameter() {defaultValue=-2, group=Bubble, parameterName=Bubble UV Depth}
  )
  n3 = ScalarParameter() {defaultValue=6, group=Bubble, parameterName=Bubble UV Tiling}
  n4 = TextureObjectParameter() {group=Bubble, parameterName=BubbleTexture, sortPriority=1, texture=/SpecterGameplay/Assets/Sprite_Ghost/Textures/T_Bubbles_A}
  n5 = Multiply(A=@n3) {constB=0.5}
  n6 = Add(A=TextureCoordinate(), B=Divide(A=Time()) {constB=120})
  n7 = Multiply(A=TextureSample(UVs=Multiply(A=@n2, B=@n3) {constB=5}, Tex=@n4).R, B=TextureSample(UVs=Multiply(A=@n5, B=@n6) {constB=5}, Tex=@n4).R)
  n8 = OneMinus(None=ScalarParameter() {defaultValue=4, group=Bubble, parameterName=Bubble Intensity, sliderMax=1})
  Divide(A=@n7, B=Multiply(A=Add(A=@n8) {constB=0.1}) {constB=10}) {constB=1}
  MaterialFunctionCall() {materialFunction=/SpecterGameplay/Gameplay/Effects/Updoot/MF_Sprite_Updoot_WPO}
  n9 = ComponentMask(None=MaterialFunctionCall(Fade Length=ScalarParameter() {defaultValue=256, parameterName=Camera Distance}) {materialFunction=&CameraDepthFade}.Result) {r=true}
  n10 = TextureSample(
   UVs=Multiply(A=TextureCoordinate(), B=ScalarParameter() {defaultValue=6, group=Bubble, parameterName=Coating UV Tiling}) {constB=12}
   Tex=@n4
  )
  n11 = Divide(A=@n10.R) {constB=20}
  n12 = Divide(A=Multiply(A=@n10.R, B=NamedRerouteUsage())) {constB=5}
  n13 = Multiply(A=NamedRerouteDeclaration(None=Add(A=@n11, B=@n12)) {name=Coating Texture, nodeColor=<rgba 0 1 0.3906 1>}, B=@n10.R)
  n14 = NamedRerouteUsage()
  n15 = Multiply(A=Add(A=Multiply(A=@n13) {constB=5}, B=Divide(A=@n14) {constB=10}), B=NamedRerouteUsage())
  Multiply(A=NamedRerouteDeclaration(None=OneMinus(None=@n9)) {name=Camera Distance Mask, nodeColor=<rgba 0 1 1 1>}, B=@n15)
  NamedRerouteDeclaration(None=VertexColor().B) {name=Face Mask, nodeColor=<rgba 0 1 0.2969 1>}
  n16 = TextureCoordinate()
  n17 = ComponentMask(None=@n16) {g=true}
  NamedRerouteDeclaration(None=Step(Y=@n17) {constX=0.25}) {name=Eye Highlight Mask, nodeColor=<rgba 0.4297 1 0 1>}
  n18 = Power(Base=DotProduct(A=CameraVectorWS(), B=VertexNormalWS())) {constExponent=5}
  n19 = VertexColor()
  n20 = Power(Base=OneMinus(None=@n19.R)) {constExponent=3}
  n21 = OneMinus(None=@n19.G)
  n22 = Add(
   A=Add(A=Multiply(A=@n20) {constB=2}, B=Power(Base=@n19.R) {constExponent=10})
   B=Add(A=Power(Base=@n21) {constExponent=10}, B=Power(Base=@n19.G) {constExponent=10})
  )
  n23 = Saturate(None=@n22)
  n24 = LinearInterpolate(
   A=LinearInterpolate(
    A=VectorParameter() {group=Color, parameterName=Main Color, sortPriority=0, defaultValue=<rgba 0.1555 0.2052 0.6094 1>}.RGB
    B=VectorParameter() {group=Color, parameterName=Core Color, sortPriority=1, defaultValue=<rgba 0.0518 0.0582 0.1094 1>}.RGB
    Alpha=Saturate(None=@n18)
   )
   B=VectorParameter() {group=Color, parameterName=Edge Color, sortPriority=2, defaultValue=<rgba 0.1909 0.3559 0.2427 1>}.RGB
   Alpha=@n23
  )
  NamedRerouteDeclaration(None=LinearInterpolate(A=@n24, B=Add(A=@n24) {constB=0.2}, Alpha=Fresnel() {baseReflectFraction=0, exponent=6})) {name=Color Combination, nodeColor=<rgba 1 0 0.2109 1>}
  n25 = DynamicParameter() {parameterIndex=2, paramNames [Eye Close, Mouth Close, Param3, Param4]}
  NamedRerouteDeclaration(None=@n25.Mouth Close) {name=Mouth Close, nodeColor=<rgba 0.8047 0 1 1>}
  NamedRerouteDeclaration(None=@n15) {name=Sugar Coating Mask, nodeColor=<rgba 0.0547 1 0 1>}
  n26 = Multiply(A=@n22) {constB=4}
  NamedRerouteDeclaration(None=OneMinus(None=Saturate(None=@n26))) {name=Center Mask, nodeColor=<rgba 0 0.0625 1 1>}
  NamedRerouteDeclaration(None=OneMinus(None=@n23)) {name=Center Mask 2, nodeColor=<rgba 0.3828 1 0 1>}
  n27 = Multiply(A=NamedRerouteUsage(), B=ScalarParameter() {defaultValue=6, group=Bubble, parameterName=Coating Intensity}) {constB=3}
  NamedRerouteDeclaration(None=Power(Base=@n27)) {name=Sugar Intensity B, nodeColor=<rgba 0.2422 0 1 1>}
  NamedRerouteDeclaration(None=@n27) {name=Sugar Intensity A, nodeColor=<rgba 0.0781 1 0 1>}
  NamedRerouteDeclaration(None=LinearInterpolate(A=NamedRerouteUsage(), B=ScalarParameter() {group=PBR, parameterName=Specular, sortPriority=1}, Alpha=NamedRerouteUsage())) {name=Specular, nodeColor=<rgba 1 0 0.3984 1>}
  n28 = Power(Base=NamedRerouteUsage()) {constExponent=3}
  NamedRerouteDeclaration(None=LinearInterpolate(
    A=ScalarParameter() {defaultValue=0.32, group=PBR, parameterName=Roughness, sortPriority=2}
    B=Saturate(None=@n28)
    Alpha=NamedRerouteUsage()
   )) {name=Roughness, nodeColor=<rgba 0 0.7187 1 1>}
  n29 = VertexColor()
  n30 = AppendVector(A=@n29.R, B=@n29.R)
  n31 = Constant() {r=0.5}
  n32 = Saturate(None=Power(Base=SphereMask(A=@n30, B=@n31) {attenuationRadius=1, hardnessPercent=0}) {constExponent=8})
  n33 = DotProduct(A=CameraVectorWS(), B=VertexNormalWS())
  n34 = Multiply(A=Multiply(A=OneMinus(None=@n29.R), B=Saturate(None=@n7)), B=Saturate(None=Power(Base=@n33) {constExponent=16}))
  NamedRerouteDeclaration(None=Multiply(A=Multiply(A=@n32, B=@n34)) {constB=100}) {name=Bubble Mask, nodeColor=<rgba 1 0 0.3516 1>}
  n35 = Add(A=NamedRerouteUsage(), B=NamedRerouteUsage())
  n36 = Add(A=@n35, B=NamedRerouteUsage())
  n37 = NamedRerouteUsage()
  n38 = LinearInterpolate(A=Add(A=@n35) {constB=0.5}, B=Add(A=Add(A=@n36, B=@n37), B=NamedRerouteUsage()), Alpha=NamedRerouteUsage())
  n39 = NamedRerouteUsage()
  n40 = NamedRerouteUsage()
  NamedRerouteDeclaration(None=Saturate(None=LinearInterpolate(A=@n38, B=@n39, Alpha=@n40) {constB=0})) {name=Base Color, nodeColor=<rgba 1 0.9609 0 1>}
  n41 = NamedRerouteUsage()
  n42 = NamedRerouteUsage()
  n43 = MaterialFunctionCall() {materialFunction=/Game/Packages/Fortress_SharedMaterials/Base_Material_Functions/SkylightColorAndIntensity.SkyLightColorAndIntensity}
  n44 = Subtract(A=Multiply(A=Multiply(A=@n42, B=@n43.Color)) {constB=0.5}, B=NamedRerouteUsage())
  NamedRerouteDeclaration(None=LinearInterpolate(A=Add(A=@n41, B=@n44), Alpha=NamedRerouteUsage()) {constB=0}) {name=Emissive, nodeColor=<rgba 0 1 0.5547 1>}
  n45 = Add(A=LinearInterpolate(A=NamedRerouteUsage(), B=NamedRerouteUsage(), Alpha=NamedRerouteUsage()), B=NamedRerouteUsage())
  n46 = Multiply(A=NamedRerouteUsage(), B=Clamp(None=NamedRerouteUsage()) {maxDefault=0.9, minDefault=0.2})
  NamedRerouteDeclaration(None=LinearInterpolate(A=Add(A=@n45, B=@n46), B=NamedRerouteUsage(), Alpha=NamedRerouteUsage())) {name=Opacity, nodeColor=<rgba 1 0.2344 0 1>}
  NamedRerouteDeclaration(None=@n25.Eye Close) {name=Eye Close, nodeColor=<rgba 0.875 1 0 1>}
  n47 = Fresnel() {baseReflectFraction=0, exponent=3}
  n48 = Add(
   A=SkyAtmosphereLightDirection()
   B=Multiply(A=MaterialFunctionCall() {materialFunction=&CameraDirectionVector}.Result) {constB=-1}
  )
  n49 = Saturate(None=DotProduct(A=PixelNormalWS(), B=Normalize(VectorInput=@n48)))
  n50 = Power(Base=@n49) {constExponent=3}
  n51 = Saturate(None=Power(Base=@n49) {constExponent=80})
  n52 = Saturate(None=Add(A=Multiply(A=@n47, B=@n50), B=Subtract(A=@n51) {constB=0.15}))
  n53 = Constant()
  NamedRerouteDeclaration(None=MaterialFunctionCall(Alpha Threshold=ShadowReplace(Default=@n52, Shadow=@n53)) {materialFunction=&DitherTemporalAA}.Result) {name=Specular Mask, nodeColor=<rgba 0 0.2266 1 1>}
  NamedRerouteDeclaration(None=Step(Y=@n17) {constX=0.5}) {name=Open Mask, nodeColor=<rgba 1 0 0.75 1>}
  n54 = TwoSidedSign()
  NamedRerouteDeclaration(None=Saturate(None=@n54)) {name=Frontface Mask, nodeColor=<rgba 0.1719 0 1 1>}
  NamedRerouteDeclaration(None=Saturate(None=Multiply(A=@n54) {constB=-1})) {name=Backface Mask, nodeColor=<rgba 0.1484 1 0 1>}
  NamedRerouteDeclaration(None=Step(Y=ComponentMask(None=@n16) {r=true}) {constX=0.5}) {name=Eye/ Mouth Mask, nodeColor=<rgba 0 1 0.9297 1>}
  n55 = NamedRerouteUsage()
  n56 = LinearInterpolate(A=OneMinus(None=@n55), B=@n55, Alpha=Step(Y=NamedRerouteUsage()) {constX=0.5})
  n57 = NamedRerouteUsage()
  n58 = NamedRerouteUsage()
  n59 = LinearInterpolate(A=OneMinus(None=@n58), B=@n58, Alpha=Step(Y=NamedRerouteUsage()) {constX=0.5})
  n60 = OneMinus(None=NamedRerouteUsage())
  NamedRerouteDeclaration(None=Add(A=Multiply(A=@n56, B=@n57), B=Multiply(A=@n59, B=@n60))) {name=Face Animation, nodeColor=<rgba 0.6875 1 0 1>}
  NamedRerouteUsage()
  NamedRerouteUsage()
  NamedRerouteUsage()
  NamedRerouteUsage()
  NamedRerouteUsage()
  ScalarParameter() {group=Face Animation, parameterName=Eye Close, sliderMax=1, sortPriority=0}
  ScalarParameter() {group=PBR, parameterName=Metallic, sortPriority=0}
  ScalarParameter() {group=Face Animation, parameterName=Mouth Close, sliderMax=1, sortPriority=1}
 ]
}
